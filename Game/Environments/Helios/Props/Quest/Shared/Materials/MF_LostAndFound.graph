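# Material function for Helios quest props that adds an optional Lost and Found scanline effect to incoming material attributes, toggled by the LAF static switch.
# Screen-space Atlas scanlines modulate base color and emissive, and a sine over time pulses desaturation on both.
# Fresnel lifts emissive, which is compensated for eye adaptation.
# Scanlines times inverse Fresnel feed MF_SortTranslucency, which drives both Opacity and Opacity Mask.
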MaterialFunction /Game/Environments/Helios/Props/Quest/Shared/Materials/MF_LostAndFound
graph {nodes [
  n1 = GetMaterialAttributes(None=FunctionInput() {inputName=OutputPin, inputType=FunctionInput_MaterialAttributes}) {attributeGetTypes [#1, #2]}
  n2 = MaterialFunctionCall!(
   Number of Scanlines=ScalarParameter() {defaultValue=0.01, group=LAF_LostAndFound, parameterName=LAF_Scanline Amount}
   Scanline Speed=ScalarParameter() {defaultValue=0.01, group=LAF_LostAndFound, parameterName=LAF_Scanline Speed}
   "Scanlines Thickness (Bias)"=ScalarParameter() {defaultValue=-0.15, group=LAF_LostAndFound, parameterName=LAF_ScanlineSize}
   Scanlines Dark Value=ScalarParameter() {defaultValue=0.5, group=LAF_LostAndFound, parameterName=LAF_Scanlines Dark}
   Scanlines Light Value=ScalarParameter() {defaultValue=1, group=LAF_LostAndFound, parameterName=LAF_Scanlines Light}
   Block Opacity=ScalarParameter() {defaultValue=0.1, group=LAF_LostAndFound, parameterName=LAF_Block Opacity}
   UVs=ScreenPosition().ViewportUV
   Time
   Block Repeat
  ) {materialFunction=/Game/Effects/Fort_Effects/Effects/Atlas/Materials/MF_Atlas_Scanlines_01}
  n3 = Divide(A=Time())
  n4 = LinearInterpolate(
   A=ScalarParameter() {defaultValue=0.1, group=LAF_LostAndFound, parameterName=LAF_DesaturateSine_Min}
   B=ScalarParameter() {defaultValue=0.6, group=LAF_LostAndFound, parameterName=LAF_DesaturateSine_Max}
   Alpha=ConstantBiasScale(None=Sine(None=@n3))
  )
  n5 = Fresnel(ExponentIn=ScalarParameter() {defaultValue=3, group=LAF_LostAndFound, parameterName=LAF_Fresnel Opacity}, Normal=VertexNormalWS()) {baseReflectFraction=0}
  n6 = Multiply(
   A=Max(A=Multiply(A=@n1.EmissiveColor, B=@n2.Result), B=@n5)
   B=ScalarParameter() {defaultValue=1, group=LAF_LostAndFound, parameterName=LAF_Emissive Increase}
  )
  n7 = OneMinus(None=@n5)
  n8 = MaterialFunctionCall!(Opacity=Multiply(A=Multiply(A=@n2.Result, B=@n7), B=ScalarParameter() {defaultValue=1.2, group=LAF_LostAndFound, parameterName=LAF_Overall Opacity})) {materialFunction=/Game/Global/Materials/MaterialGym/Materials/SortedOutline/MF_SortTranslucency}
  n9 = SetMaterialAttributes!(
   MaterialAttributes=@n1.MaterialAttributes
   BaseColor=Desaturation(None=Multiply(A=@n1.BaseColor, B=@n2.Result), Fraction=@n4)
   EmissiveColor=EyeAdaptationInverse(LightValueInput=Desaturation(None=@n6, Fraction=@n4), AlphaInput=Constant() {r=1}).EyeAdaptationInverse
   Opacity=@n8.Result
   OpacityMask=@n8.Result
  ) {attributeSetTypes [
    #1
    #2
    #3
    #4
   ]}
  FunctionOutput(None=StaticSwitchParameter(True=@n9, False=@n1.MaterialAttributes) {group=LAF_LostAndFound, parameterName=LAF, sortPriority=-1}) {outputName="Switch Param (False)"}
 ]}
